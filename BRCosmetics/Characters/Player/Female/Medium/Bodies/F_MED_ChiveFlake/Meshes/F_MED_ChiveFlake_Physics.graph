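# Physics asset for the F_MED_ChiveFlake player body cosmetic, set up for skirt physics with capsules on the skirt, pelvis, lower spine and legs.
# Eight three-bone skirt chains hang from base bodies linked to spine_01, with tight twist limits and locked swing1 on front and side chains.
# Free-motion constraints cross-link lower side and front segments.
# Left and right differ: only the left links side to back, and back-right swing2 is tighter.
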PhysicsAsset /BRCosmetics/Characters/Player/Female/Medium/Bodies/F_MED_ChiveFlake/Meshes/F_MED_ChiveFlake_Physics
bodies {
 base_skirt_bk_mid [{shapeType=Capsule, length=1, radius=1, shapeName=base_skirt_bk_mid_capsule, center=<1 0 0>, rotation=<rot -90 0 0>}]
 base_skirt_frnt_mid [{shapeType=Capsule, length=1, radius=1, shapeName=base_skirt_frnt_mid_capsule, center=<1 0 0>, rotation=<rot -90 0 0>}]
 base_skirt_side_bk_l [{shapeType=Capsule, length=1, radius=1, shapeName=base_skirt_side_bk_l_capsule, center=<1 0 0>, rotation=<rot -90 0 0>}]
 base_skirt_side_bk_r [{shapeType=Capsule, length=1, radius=1, shapeName=base_skirt_side_bk_r_capsule, center=<1 0 0>, rotation=<rot -90 0 0>}]
 base_skirt_side_frnt_l [{shapeType=Capsule, length=1, radius=1, shapeName=base_skirt_side_frnt_l_capsule, center=<1 0 0>, rotation=<rot -90 0 0>}]
 base_skirt_side_frnt_r [{shapeType=Capsule, length=1, radius=1, shapeName=base_skirt_side_frnt_r_capsule, center=<1 0 0>, rotation=<rot -90 0 0>}]
 base_skirt_side_l [{shapeType=Capsule, length=1, radius=1, shapeName=base_skirt_side_l_capsule, center=<1 0 0>, rotation=<rot -90 0 0>}]
 base_skirt_side_r [{shapeType=Capsule, length=1, radius=1, shapeName=base_skirt_side_r_capsule, center=<1 0 0>, rotation=<rot -90 0 0>}]
 calf_l [{shapeType=Capsule, length=30.541, radius=5, shapeName=thigh_l_capsule, center=<-20 -1.016 -1.6445>, rotation=<rot 0.0806 -90.8115 -89.0784>}]
 calf_r [{shapeType=Capsule, length=30.541, radius=5, shapeName=thigh_r_capsule, center=<20 0.7739 1.37>, rotation=<rot 0.0806 -90.8115 -89.0784>}]
 dyn_skirt_bk_mid_01 [{shapeType=Capsule, length=15, radius=4, center=<10 0 0>, rotation=<rot 0 0 90>}]
 dyn_skirt_bk_mid_02 [{shapeType=Capsule, length=15, radius=7, center=<10 0 0>, rotation=<rot 0 0 90>}]
 dyn_skirt_bk_mid_03 [{shapeType=Capsule, length=25, radius=10, center=<10 0 0>, rotation=<rot 0 0 90>}]
 dyn_skirt_frnt_mid_01 [{shapeType=Capsule, length=15, radius=4, center=<15 0 5>, rotation=<rot 0 0 90>}]
 dyn_skirt_frnt_mid_02 [{shapeType=Capsule, length=15, radius=5, center=<10 0 5>, rotation=<rot 0 0 90>}]
 dyn_skirt_frnt_mid_03 [{shapeType=Capsule, length=15, radius=10, center=<8.9327 0 2.5>, rotation=<rot 0 0 90>}]
 dyn_skirt_side_bk_l_01 [{shapeType=Capsule, length=15, radius=4, center=<10 0 2.2>, rotation=<rot 0 0 90>}]
 dyn_skirt_side_bk_l_02 [{shapeType=Capsule, length=15, radius=7, center=<10 0 0>, rotation=<rot 0 0 90>}]
 dyn_skirt_side_bk_l_03 [{shapeType=Capsule, length=15, radius=10, center=<10 0 2.5>, rotation=<rot 0 0 90>}]
 dyn_skirt_side_bk_r_01 [{shapeType=Capsule, length=15, radius=4, center=<10 0 2.2>, rotation=<rot 0 0 90>}]
 dyn_skirt_side_bk_r_02 [{shapeType=Capsule, length=15, radius=7, center=<10 0 0>, rotation=<rot 0 0 90>}]
 dyn_skirt_side_bk_r_03 [{shapeType=Capsule, length=15, radius=10, center=<10 0 2.5>, rotation=<rot 0 0 90>}]
 dyn_skirt_side_frnt_l_01 [{shapeType=Capsule, length=15, radius=4, center=<10 0 5>, rotation=<rot 0 0 90>}]
 dyn_skirt_side_frnt_l_02 [{shapeType=Capsule, length=15, radius=10, center=<10 0 6>, rotation=<rot 0 0 90>}]
 dyn_skirt_side_frnt_l_03 [{shapeType=Capsule, length=15, radius=10, center=<10 0 10>, rotation=<rot 0 0 89>}]
 dyn_skirt_side_frnt_r_01 [{shapeType=Capsule, length=15, radius=4, center=<10 0 5>, rotation=<rot 0 0 90>}]
 dyn_skirt_side_frnt_r_02 [{shapeType=Capsule, length=15, radius=10, center=<10 0 6>, rotation=<rot 0 0 90>}]
 dyn_skirt_side_frnt_r_03 [{shapeType=Capsule, length=15, radius=10, center=<10 0 5>, rotation=<rot 0 0 90>}]
 dyn_skirt_side_l_01 [{shapeType=Capsule, length=15, radius=4, center=<10 0 2.7595>, rotation=<rot 0 0 90>}]
 dyn_skirt_side_l_02 [{shapeType=Capsule, length=15, radius=7, center=<10 0 2.5>, rotation=<rot 0 0 90>}]
 dyn_skirt_side_l_03 [{shapeType=Capsule, length=15, radius=10, center=<10 0 10>, rotation=<rot 0 0 90>}]
 dyn_skirt_side_r_01 [{shapeType=Capsule, length=15, radius=4, center=<10 0 2.9063>, rotation=<rot 0 0 90>}]
 dyn_skirt_side_r_02 [{shapeType=Capsule, length=15, radius=7, center=<10 0 2.5917>, rotation=<rot 0 0 90>}]
 dyn_skirt_side_r_03 [{shapeType=Capsule, length=15, radius=10, center=<10 0 10>, rotation=<rot 0 0 90>}]
 foot_l [{shapeType=Capsule, length=10, radius=10, shapeName=foot_r_capsule, center=<0 9.806 0>, rotation=<rot 0 0 -90>}]
 foot_r [{shapeType=Capsule, length=10, radius=10, shapeName=foot_r_capsule, center=<-0.0001 -7.8605 0>, rotation=<rot 0 -0.0005 -90>}]
 pelvis [{shapeType=Capsule, length=10.7667, radius=9, shapeName=pelvis_capsule, center=<0 -0.6252 0>, rotation=<rot 0 -90.0003 0>}]
 spine_01 [{shapeType=Capsule, length=10.7667, radius=9, shapeName=pelvis_capsule, center=<-0.4976 -0.6252 0>, rotation=<rot 0 -90.0003 0>}]
 thigh_l [
  {
   shapeType=Capsule
   length=28.1684
   radius=4.4063
   shapeName=thigh_l_capsule
   center=<-28.1669 -0.895 -1.5072>
   rotation=<rot 0 -85.8121 -89.0749>
  }
  {
   shapeType=Capsule
   length=28.1684
   radius=4.4063
   shapeName=thigh_l_capsule
   center=<-28.2389 1.8513 6.4656>
   rotation=<rot 0 -85.8121 -89.0749>
  }
 ]
 thigh_r [
  {
   shapeType=Capsule
   length=28.1684
   radius=4.4063
   shapeName=thigh_r_capsule
   center=<28.1669 0.895 1.5072>
   rotation=<rot 0 -85.8121 -89.0749>
  }
  {
   shapeType=Capsule
   length=28.1684
   radius=4.4063
   shapeName=thigh_r_capsule
   center=<28.2648 -2.2982 -6.8906>
   rotation=<rot 0 -85.8121 -89.0749>
  }
 ]
}
constraints [
 {bone1Name=dyn_skirt_bk_mid_01, bone2Name=base_skirt_bk_mid, swing1LimitDegrees=5, swing2LimitDegrees=35, twistLimitDegrees=1}
 {bone1Name=dyn_skirt_bk_mid_02, bone2Name=dyn_skirt_bk_mid_01, swing1LimitDegrees=5, swing2LimitDegrees=35, twistLimitDegrees=1}
 {bone1Name=dyn_skirt_bk_mid_03, bone2Name=dyn_skirt_bk_mid_02, swing1LimitDegrees=5, swing2LimitDegrees=35, twistLimitDegrees=1}
 {
  bone1Name=dyn_skirt_frnt_mid_01
  bone2Name=base_skirt_frnt_mid
  swing1LimitDegrees=25
  swing1Motion=Locked
  swing2LimitDegrees=35
  twistLimitDegrees=1
 }
 {
  bone1Name=dyn_skirt_frnt_mid_02
  bone2Name=dyn_skirt_frnt_mid_01
  swing1LimitDegrees=25
  swing1Motion=Locked
  swing2LimitDegrees=35
  twistLimitDegrees=1
 }
 {
  bone1Name=dyn_skirt_frnt_mid_03
  bone2Name=dyn_skirt_frnt_mid_02
  swing1LimitDegrees=25
  swing1Motion=Locked
  swing2LimitDegrees=35
  twistLimitDegrees=1
 }
 {bone1Name=dyn_skirt_side_bk_l_01, bone2Name=base_skirt_side_bk_l, swing1LimitDegrees=5, swing2LimitDegrees=35, twistLimitDegrees=1}
 {bone1Name=dyn_skirt_side_bk_l_02, bone2Name=dyn_skirt_side_bk_l_01, swing1LimitDegrees=5, swing2LimitDegrees=35, twistLimitDegrees=1}
 {bone1Name=dyn_skirt_side_bk_l_03, bone2Name=dyn_skirt_side_bk_l_02, swing1LimitDegrees=5, swing2LimitDegrees=35, twistLimitDegrees=1}
 {bone1Name=dyn_skirt_side_bk_r_01, bone2Name=base_skirt_side_bk_r, swing1LimitDegrees=5, swing2LimitDegrees=17.5, twistLimitDegrees=1}
 {bone1Name=dyn_skirt_side_bk_r_02, bone2Name=dyn_skirt_side_bk_r_01, swing1LimitDegrees=5, swing2LimitDegrees=17.5, twistLimitDegrees=1}
 {bone1Name=dyn_skirt_side_bk_r_03, bone2Name=dyn_skirt_side_bk_r_02, swing1LimitDegrees=5, swing2LimitDegrees=17.5, twistLimitDegrees=1}
 {
  bone1Name=dyn_skirt_side_frnt_l_01
  bone2Name=base_skirt_side_frnt_l
  swing1LimitDegrees=10
  swing1Motion=Locked
  swing2LimitDegrees=20
  twistLimitDegrees=1
 }
 {
  bone1Name=dyn_skirt_side_frnt_l_02
  bone2Name=dyn_skirt_side_frnt_l_01
  swing1LimitDegrees=10
  swing1Motion=Locked
  swing2LimitDegrees=30
  twistLimitDegrees=1
 }
 {
  bone1Name=dyn_skirt_side_frnt_l_03
  bone2Name=dyn_skirt_side_frnt_l_02
  swing1LimitDegrees=10
  swing1Motion=Locked
  swing2LimitDegrees=30
  twistLimitDegrees=1
 }
 {
  bone1Name=dyn_skirt_side_frnt_r_01
  bone2Name=base_skirt_side_frnt_r
  swing1LimitDegrees=10
  swing1Motion=Locked
  swing2LimitDegrees=20
  twistLimitDegrees=1
 }
 {
  bone1Name=dyn_skirt_side_frnt_r_02
  bone2Name=dyn_skirt_side_frnt_r_01
  swing1LimitDegrees=10
  swing1Motion=Locked
  swing2LimitDegrees=30
  twistLimitDegrees=1
 }
 {
  bone1Name=dyn_skirt_side_frnt_r_03
  bone2Name=dyn_skirt_side_frnt_r_02
  swing1LimitDegrees=10
  swing1Motion=Locked
  swing2LimitDegrees=30
  twistLimitDegrees=1
 }
 {
  bone1Name=dyn_skirt_side_l_01
  bone2Name=base_skirt_side_l
  swing1LimitDegrees=25
  swing1Motion=Locked
  swing2LimitDegrees=35
  twistLimitDegrees=1
 }
 {
  bone1Name=dyn_skirt_side_l_02
  bone2Name=dyn_skirt_side_l_01
  swing1LimitDegrees=25
  swing1Motion=Locked
  swing2LimitDegrees=35
  twistLimitDegrees=1
 }
 {
  bone1Name=dyn_skirt_side_l_03
  bone2Name=dyn_skirt_side_l_02
  swing1LimitDegrees=25
  swing1Motion=Locked
  swing2LimitDegrees=35
  twistLimitDegrees=1
 }
 {
  bone1Name=dyn_skirt_side_r_01
  bone2Name=base_skirt_side_r
  swing1LimitDegrees=25
  swing1Motion=Locked
  swing2LimitDegrees=35
  twistLimitDegrees=1
 }
 {
  bone1Name=dyn_skirt_side_r_02
  bone2Name=dyn_skirt_side_r_01
  swing1LimitDegrees=25
  swing1Motion=Locked
  swing2LimitDegrees=35
  twistLimitDegrees=1
 }
 {
  bone1Name=dyn_skirt_side_r_03
  bone2Name=dyn_skirt_side_r_02
  swing1LimitDegrees=25
  swing1Motion=Locked
  swing2LimitDegrees=35
  twistLimitDegrees=1
 }
 {
  bone1Name=base_skirt_side_bk_l
  bone2Name=spine_01
  swing1LimitDegrees=45
  swing1Motion=Free
  swing2LimitDegrees=45
  swing2Motion=Free
  twistLimitDegrees=45
  twistMotion=Free
 }
 {
  bone1Name=base_skirt_side_frnt_l
  bone2Name=spine_01
  swing1LimitDegrees=45
  swing1Motion=Free
  swing2LimitDegrees=45
  swing2Motion=Free
  twistLimitDegrees=45
  twistMotion=Free
 }
 {
  bone1Name=base_skirt_bk_mid
  bone2Name=spine_01
  swing1LimitDegrees=45
  swing1Motion=Free
  swing2LimitDegrees=45
  swing2Motion=Free
  twistLimitDegrees=45
  twistMotion=Free
 }
 {
  bone1Name=base_skirt_frnt_mid
  bone2Name=spine_01
  swing1LimitDegrees=45
  swing1Motion=Free
  swing2LimitDegrees=45
  swing2Motion=Free
  twistLimitDegrees=45
  twistMotion=Free
 }
 {
  bone1Name=base_skirt_side_l
  bone2Name=spine_01
  swing1LimitDegrees=45
  swing1Motion=Free
  swing2LimitDegrees=45
  swing2Motion=Free
  twistLimitDegrees=45
  twistMotion=Free
 }
 {
  bone1Name=base_skirt_side_bk_r
  bone2Name=spine_01
  swing1LimitDegrees=45
  swing1Motion=Free
  swing2LimitDegrees=45
  swing2Motion=Free
  twistLimitDegrees=45
  twistMotion=Free
 }
 {
  bone1Name=base_skirt_side_frnt_r
  bone2Name=spine_01
  swing1LimitDegrees=45
  swing1Motion=Free
  swing2LimitDegrees=45
  swing2Motion=Free
  twistLimitDegrees=45
  twistMotion=Free
 }
 {
  bone1Name=base_skirt_side_r
  bone2Name=spine_01
  swing1LimitDegrees=45
  swing1Motion=Free
  swing2LimitDegrees=45
  swing2Motion=Free
  twistLimitDegrees=45
  twistMotion=Free
 }
 {
  bone1Name=thigh_l
  bone2Name=pelvis
  swing1LimitDegrees=45
  swing1Motion=Free
  swing2LimitDegrees=45
  swing2Motion=Free
  twistLimitDegrees=45
  twistMotion=Free
 }
 {
  bone1Name=calf_l
  bone2Name=thigh_l
  swing1LimitDegrees=45
  swing1Motion=Free
  swing2LimitDegrees=45
  swing2Motion=Free
  twistLimitDegrees=45
  twistMotion=Free
 }
 {
  bone1Name=thigh_r
  bone2Name=pelvis
  swing1LimitDegrees=45
  swing1Motion=Free
  swing2LimitDegrees=45
  swing2Motion=Free
  twistLimitDegrees=45
  twistMotion=Free
 }
 {
  bone1Name=calf_r
  bone2Name=thigh_r
  swing1LimitDegrees=45
  swing1Motion=Free
  swing2LimitDegrees=45
  swing2Motion=Free
  twistLimitDegrees=45
  twistMotion=Free
 }
 {
  bone1Name=foot_l
  bone2Name=calf_l
  swing1LimitDegrees=45
  swing1Motion=Free
  swing2LimitDegrees=45
  swing2Motion=Free
  twistLimitDegrees=45
  twistMotion=Free
 }
 {
  bone1Name=foot_r
  bone2Name=calf_r
  swing1LimitDegrees=45
  swing1Motion=Free
  swing2LimitDegrees=45
  swing2Motion=Free
  twistLimitDegrees=45
  twistMotion=Free
 }
 {
  bone1Name=dyn_skirt_side_r_02
  bone2Name=dyn_skirt_side_frnt_r_02
  swing1LimitDegrees=45
  swing1Motion=Free
  swing2LimitDegrees=45
  swing2Motion=Free
  twistLimitDegrees=45
  twistMotion=Free
 }
 {
  bone1Name=dyn_skirt_side_r_03
  bone2Name=dyn_skirt_side_frnt_r_03
  swing1LimitDegrees=45
  swing1Motion=Free
  swing2LimitDegrees=45
  swing2Motion=Free
  twistLimitDegrees=45
  twistMotion=Free
 }
 {
  bone1Name=dyn_skirt_side_l_02
  bone2Name=dyn_skirt_side_frnt_l_02
  swing1LimitDegrees=45
  swing1Motion=Free
  swing2LimitDegrees=45
  swing2Motion=Free
  twistLimitDegrees=45
  twistMotion=Free
 }
 {
  bone1Name=dyn_skirt_side_l_03
  bone2Name=dyn_skirt_side_frnt_l_03
  swing1LimitDegrees=45
  swing1Motion=Free
  swing2LimitDegrees=45
  swing2Motion=Free
  twistLimitDegrees=45
  twistMotion=Free
 }
 {
  bone1Name=dyn_skirt_side_l_02
  bone2Name=dyn_skirt_side_bk_l_02
  swing1LimitDegrees=45
  swing1Motion=Free
  swing2LimitDegrees=45
  swing2Motion=Free
  twistLimitDegrees=45
  twistMotion=Free
 }
 {
  bone1Name=spine_01
  bone2Name=pelvis
  swing1LimitDegrees=45
  swing1Motion=Free
  swing2LimitDegrees=45
  swing2Motion=Free
  twistLimitDegrees=45
  twistMotion=Free
 }
]
