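# Post-process style AnimBlueprint for the F_MED_ChiveFlake female cosmetic body, adding secondary jacket, hood and skirt motion to the copied parent-mesh pose.
# A deform Control Rig feeds AnimDynamics, a pelvis-space RigidBody, corrective bone transforms, then a post-deform Control Rig.
# The event graph derives gravity, physics alpha and corrective weights from movement state, thigh and head rotations, and backbling.
# A CajunTaco animation curve drives the post-deform rig's emote inputs.
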AnimBlueprint /BRCosmetics/Characters/Player/Female/Medium/Bodies/F_MED_ChiveFlake/Meshes/F_MED_ChiveFlake_AnimBP
target_skeleton=&/F_MED_ChiveFlake_Skeleton
graphs [
 AnimGraph {nodes [
   n1 = Control Rig!(
    Source=Use cached pose 'Deform_Cache'()
    Alpha
    Is Cajuntaco Emote:bool=Get Is Cajuntaco Emote() -> (Is Cajuntaco Emote:bool)
    Custom Emote Float:double=Get CustomEmoteFloat() -> (CustomEmoteFloat:double)
   ) -> (Pose) {
    alphaBoolBlend {blendOption=Linear}
    alphaScaleBiasClamp {bInterpResult=true, interpSpeedDecreasing=5, interpSpeedIncreasing=5}
    controlRigAssetReference {blueprintRigClass=&/F_MED_ChiveFlake_PostDeform_CtrlRig.F_MED_ChiveFlake_PostDeform_CtrlRig_C}
   }
   Output Pose(Result=@n1) {layerGroup=None}
   n2 = Control Rig!(
    Source=Copy Pose From Mesh() {bCopyCurves=true, bUseAttachedParent=true}
    Alpha
    is skydiving:double=Get skydiving() -> (skydiving:double)
    isCrouching:bool=Get bIsCrouching()
    isCrouchingMoving:double=Get crouchingMoving() -> (crouchingMoving:double)
    isSlidingwithPickAxe:bool=Get bIsKneeSliding()
    RBAN_on:bool=Get bRigidBodyEnabled()
    isCrouchingNotMoving:bool=Get isCrounchingNotMoving() -> (isCrounchingNotMoving:bool)
    isSliding:bool=Get bIsSliding()
    DBNO:bool=Get bIsDBNO()
    DBNOfloat:double=Get dbnoFloat() -> (dbnoFloat:double)
    IsPlayingCustomEmote:bool=Get Is Custom Emote Playing() -> (Is Custom Emote Playing:bool)
   ) -> (Pose) {
    alphaBoolBlend {blendOption=Linear}
    controlRigAssetReference {blueprintRigClass=&/F_MED_ChiveFlake_DeformCtrlRig.F_MED_ChiveFlake_DeformCtrlRig_C}
   }
   Save cached pose 'CtrlRig_Cache'(Pose=@n2) {cacheName=CtrlRig_Cache}
   n3 = Anim Dynamics - Bone: dyn_jacket_frnt_r(ComponentPose=Local To Component(LocalPose=Use cached pose 'CtrlRig_Cache'()), Alpha=1, ExternalForce="0,0,0") {
    angularDampingOverride=1
    angularSpringConstant=5000
    bAngularSpring=true
    bOverrideAngularDamping=true
    bUseGravityOverride=true
    gravityScale=0.35
    simulationSpace=BoneRelative
    alphaBoolBlend {blendOption=Linear}
    boundBone {boneName=dyn_jacket_frnt_r}
    componentAppliedLinearAccClamp=<2500 2500 1500>
    componentLinearAccScale=<1 1 1>
    componentLinearVelScale=<1 1 1>
    gravityOverride {z=-980}
    physicsBodyDefinitions [{
      boundBone {boneName=dyn_jacket_frnt_r}
      boxExtents=<25 25 25>
      constraintSetup {angularLimitsMax=<30 0 30>, angularLimitsMin=<-30 -30 0>, angularTarget=<1 0 0>}
      localJointOffset=<3 0 0>
     }]
    relativeSpaceBone {boneName=root}
   }
   n4 = Anim Dynamics - Bone: dyn_jacket_side_r(ComponentPose=@n3, Alpha=1, ExternalForce="0,0,0") {
    angularDampingOverride=1
    angularSpringConstant=5000
    bAngularSpring=true
    bOverrideAngularDamping=true
    bUseGravityOverride=true
    gravityScale=0.35
    simulationSpace=BoneRelative
    alphaBoolBlend {blendOption=Linear}
    boundBone {boneName=dyn_jacket_side_r}
    componentAppliedLinearAccClamp=<2500 2500 1500>
    componentLinearAccScale=<1 1 1>
    componentLinearVelScale=<1 1 1>
    gravityOverride {z=-980}
    physicsBodyDefinitions [{
      boundBone {boneName=dyn_jacket_side_r}
      boxExtents=<25 25 25>
      constraintSetup {angularLimitsMax=<30 0 30>, angularLimitsMin=<-30 -30 0>, angularTarget=<1 0 0>}
      localJointOffset=<3 0 0>
     }]
    relativeSpaceBone {boneName=root}
   }
   n5 = Anim Dynamics - Bone: dyn_jacket_back_r(ComponentPose=@n4, Alpha=1, ExternalForce="0,0,0") {
    angularDampingOverride=1
    angularSpringConstant=5000
    bAngularSpring=true
    bOverrideAngularDamping=true
    bUseGravityOverride=true
    gravityScale=0.35
    simulationSpace=BoneRelative
    alphaBoolBlend {blendOption=Linear}
    boundBone {boneName=dyn_jacket_back_r}
    componentAppliedLinearAccClamp=<2500 2500 1500>
    componentLinearAccScale=<1 1 1>
    componentLinearVelScale=<1 1 1>
    gravityOverride {z=-980}
    physicsBodyDefinitions [{
      boundBone {boneName=dyn_jacket_back_r}
      boxExtents=<25 25 25>
      constraintSetup {angularLimitsMax=<30 0 30>, angularLimitsMin=<-30 -30 0>, angularTarget=<1 0 0>}
      localJointOffset=<3 0 0>
     }]
    relativeSpaceBone {boneName=root}
   }
   n6 = Anim Dynamics - Bone: dyn_jacket_frnt_l(ComponentPose=@n5, Alpha=1, ExternalForce="0,0,0") {
    angularDampingOverride=1
    angularSpringConstant=5000
    bAngularSpring=true
    bOverrideAngularDamping=true
    bUseGravityOverride=true
    gravityScale=0.35
    simulationSpace=BoneRelative
    alphaBoolBlend {blendOption=Linear}
    boundBone {boneName=dyn_jacket_frnt_l}
    componentAppliedLinearAccClamp=<2500 2500 1500>
    componentLinearAccScale=<1 1 1>
    componentLinearVelScale=<1 1 1>
    gravityOverride {z=-980}
    physicsBodyDefinitions [{
      boundBone {boneName=dyn_jacket_frnt_l}
      boxExtents=<25 25 25>
      constraintSetup {angularLimitsMax=<30 0 30>, angularLimitsMin=<-30 -30 0>, angularTarget=<1 0 0>}
      localJointOffset=<3 0 0>
     }]
    relativeSpaceBone {boneName=root}
   }
   n7 = Anim Dynamics - Bone: dyn_jacket_side_l(ComponentPose=@n6, Alpha=1, ExternalForce="0,0,0") {
    angularDampingOverride=1
    angularSpringConstant=5000
    bAngularSpring=true
    bOverrideAngularDamping=true
    bUseGravityOverride=true
    gravityScale=0.35
    simulationSpace=BoneRelative
    alphaBoolBlend {blendOption=Linear}
    boundBone {boneName=dyn_jacket_side_l}
    componentAppliedLinearAccClamp=<2500 2500 1500>
    componentLinearAccScale=<1 1 1>
    componentLinearVelScale=<1 1 1>
    gravityOverride {z=-980}
    physicsBodyDefinitions [{
      boundBone {boneName=dyn_jacket_side_l}
      boxExtents=<25 25 25>
      constraintSetup {angularLimitsMax=<30 0 30>, angularLimitsMin=<-30 -30 0>, angularTarget=<1 0 0>}
      localJointOffset=<3 0 0>
     }]
    relativeSpaceBone {boneName=root}
   }
   n8 = Anim Dynamics - Bone: dyn_jacket_back_l(ComponentPose=@n7, Alpha=1, ExternalForce="0,0,0") {
    angularDampingOverride=1
    angularSpringConstant=5000
    bAngularSpring=true
    bOverrideAngularDamping=true
    bUseGravityOverride=true
    gravityScale=0.35
    simulationSpace=BoneRelative
    alphaBoolBlend {blendOption=Linear}
    boundBone {boneName=dyn_jacket_back_l}
    componentAppliedLinearAccClamp=<2500 2500 1500>
    componentLinearAccScale=<1 1 1>
    componentLinearVelScale=<1 1 1>
    gravityOverride {z=-980}
    physicsBodyDefinitions [{
      boundBone {boneName=dyn_jacket_back_l}
      boxExtents=<25 25 25>
      constraintSetup {angularLimitsMax=<30 0 30>, angularLimitsMin=<-30 -30 0>, angularTarget=<1 0 0>}
      localJointOffset=<3 0 0>
     }]
    relativeSpaceBone {boneName=root}
   }
   n9 = Anim Dynamics - Bone: dyn_top_hood(ComponentPose=@n8, Alpha=1, ExternalForce="0,0,0") {
    angularDampingOverride=1
    angularSpringConstant=5000
    bAngularSpring=true
    bOverrideAngularDamping=true
    bUseGravityOverride=true
    gravityScale=0.35
    simulationSpace=RootRelative
    alphaBoolBlend {blendOption=Linear}
    boundBone {boneName=dyn_top_hood}
    componentAppliedLinearAccClamp=<2500 2500 1500>
    componentLinearAccScale=<1 1 1>
    componentLinearVelScale=<1 1 1>
    gravityOverride {z=-980}
    physicsBodyDefinitions [{
      boundBone {boneName=dyn_top_hood}
      boxExtents=<35 35 35>
      constraintSetup {angularLimitsMax=<15 0 15>, angularLimitsMin=<-15 -30 -15>, angularTarget=<1 0 0>}
      localJointOffset=<4 0 0>
     }]
   }
   n10 = "Transform (Modify) Bone - Bone: base_skirt_side_bk_r"!(ComponentPose=@n9, bAlphaBoolEnabled:bool=True, Translation="-5,0,0", Rotation="10,15,0", Scale="1,1,1") -> (Pose) {
    alphaInputType=Bool
    rotationMode=BMM_Additive
    rotationSpace=BCS_BoneSpace
    scaleMode=BMM_Additive
    scaleSpace=BCS_BoneSpace
    translationMode=BMM_Additive
    translationSpace=BCS_BoneSpace
    alphaBoolBlend {blendOption=Linear, blendOutTime=0.2}
    boneToModify {boneName=base_skirt_side_bk_r}
    rotation {pitch=10, yaw=15}
    translation {x=-5}
   }
   Save cached pose 'animDyn_Cache'(Pose=Component To Local(ComponentPose=@n10)) {cacheName=animDyn_Cache}
   n11 = RigidBody!(
    ComponentPose=Local To Component(LocalPose=Use cached pose 'animDyn_Cache'())
    Alpha=Get RB alpha() -> (RB alpha:double)
    OverrideWorldGravity:Vector=Get GravityOverride()
    ExternalForce
   ) -> (Pose) {
    bClampLinearTranslationLimitToRefPose=true
    bFreezeIncomingPoseOnStart=true
    bOverrideWorldGravity=true
    bUseExternalClothCollision=true
    overridePhysicsAsset=&/F_MED_ChiveFlake_Physics
    simulationSpace=BaseBoneSpace
    alphaBoolBlend {blendOption=Linear}
    baseBoneRef {boneName=pelvis}
    componentAppliedLinearAccClamp=<3000 3000 3000>
    overrideWorldGravity {z=-980}
    simSpaceSettings {maxAngularAcceleration=100, maxAngularVelocity=9, maxLinearAcceleration=300, maxLinearVelocity=400, velocityScaleZ=0.12, worldAlpha=1}
   }
   n12 = "Transform (Modify) Bone - Bone: base_skirt_side_frnt_l"(ComponentPose=@n11, Alpha=Get isCrouchInterp() -> (isCrouchInterp:double), Translation="0,0,0", Rotation="15,0,0", Scale="1,1,1") {
    rotationMode=BMM_Additive
    rotationSpace=BCS_BoneSpace
    translationMode=BMM_Additive
    translationSpace=BCS_BoneSpace
    alphaBoolBlend {blendOption=Linear}
    alphaScaleBiasClamp {interpSpeedDecreasing=0.1, interpSpeedIncreasing=0.1}
    boneToModify {boneName=base_skirt_side_frnt_l}
    rotation {pitch=15}
   }
   n13 = Get leg_r_rot() -> (leg_r_rot:double)
   n14 = "Transform (Modify) Bone - Bone: base_skirt_side_frnt_r"(ComponentPose=@n12, Alpha=@n13, Translation="0,0,0", Rotation=Get rotationRside() -> (rotationRside:Rotator), Scale="1,1,1") {
    rotationMode=BMM_Additive
    rotationSpace=BCS_BoneSpace
    translationMode=BMM_Additive
    translationSpace=BCS_BoneSpace
    alphaBoolBlend {blendOption=Linear}
    boneToModify {boneName=base_skirt_side_frnt_r}
   }
   n15 = "Transform (Modify) Bone - Bone: base_skirt_side_r"(ComponentPose=@n14, Alpha=@n13, Translation="0,0,0", Rotation="20,0,0", Scale="1,1,1") {
    rotationMode=BMM_Additive
    rotationSpace=BCS_BoneSpace
    translationMode=BMM_Additive
    translationSpace=BCS_BoneSpace
    alphaBoolBlend {blendOption=Linear}
    boneToModify {boneName=base_skirt_side_r}
    rotation {pitch=20}
   }
   n16 = Get leg_l_rot() -> (leg_l_rot:double)
   n17 = "Transform (Modify) Bone - Bone: base_skirt_side_frnt_l"(ComponentPose=@n15, Alpha=@n16, Translation="0,0,0", Rotation="45,0,-45", Scale="1,1,1") {
    rotationMode=BMM_Additive
    rotationSpace=BCS_BoneSpace
    translationMode=BMM_Additive
    translationSpace=BCS_BoneSpace
    alphaBoolBlend {blendOption=Linear}
    boneToModify {boneName=base_skirt_side_frnt_l}
    rotation {pitch=45, roll=-45}
   }
   n18 = "Transform (Modify) Bone - Bone: base_skirt_side_l"(ComponentPose=@n17, Alpha=@n16, Translation="0,0,0", Rotation="20,0,0", Scale="1,1,1") {
    rotationMode=BMM_Additive
    rotationSpace=BCS_BoneSpace
    translationMode=BMM_Additive
    translationSpace=BCS_BoneSpace
    alphaBoolBlend {blendOption=Linear}
    boneToModify {boneName=base_skirt_side_l}
    rotation {pitch=20}
   }
   n19 = "Transform (Modify) Bone - Bone: base_bottom_hood"(ComponentPose=@n18, Alpha=Get headUp_BackBlink() -> (headUp_BackBlink:double), Translation="0,0,0", Rotation="0,0,0", Scale="1,1,0.75") {
    rotationMode=BMM_Additive
    rotationSpace=BCS_BoneSpace
    scaleMode=BMM_Additive
    scaleSpace=BCS_BoneSpace
    translationMode=BMM_Additive
    translationSpace=BCS_BoneSpace
    alphaBoolBlend {blendInTime=0.2, blendOption=Linear, blendOutTime=0.2}
    boneToModify {boneName=base_bottom_hood}
    scale {z=0.75}
   }
   Save cached pose 'Deform_Cache'(Pose=Component To Local(ComponentPose=@n19)) {cacheName=Deform_Cache}
  ]}
 EventGraph {nodes [
   n1 = Event BlueprintUpdateAnimation()
   > Set DeltaTimeLocal(DeltaTimeLocal:double=@n1.DeltaTimeX) -> (Output_Get:double)
   > n2 = Sequence!(execute) -> (then_0, then_1, then_2:Exec, then_3:Exec, then_4:Exec, then_5:Exec)
   > n3 = Gravity Override State Function DEPRECATED(CharacterPartAnimInstance=Self-Reference(), GameTimeInSeconds=GetGameTimeInSeconds(), GravityOverrideParams=Get Gravity Override Params())
   > Set GravityOverride(GravityOverride=@n3)
   n4 = Get bIsSkyDiving()
   > Set skydiving(skydiving:double=FInterpTo(Current=Get skydiving(), Target="To Float (Boolean)"(InBool=@n4), DeltaTime=Get DeltaTimeLocal() -> (DeltaTimeLocal:double), InterpSpeed=2)) -> (Output_Get:double)
   n5 = AND Boolean(A=Get bIsCrouching(), B=Get bIsMoving2D())
   > Set crouchingMoving(crouchingMoving:double=FInterpTo(Current=Get crouchingMoving(), Target="To Float (Boolean)"(InBool=@n5), DeltaTime=Get DeltaTimeLocal(), InterpSpeed=2)) -> (Output_Get:double)
   > Set dbnoFloat(dbnoFloat:double="To Float (Boolean)"(InBool=Get bIsDBNO())) -> (Output_Get:double)
   > n6 = Branch(Condition=OR Boolean(A=Get bIsTacticalSprinting(), B=Get bIsDBNO()))
   > Set RB alpha(RB alpha:double=0.5) -> (Output_Get:double)
   n7 = Branch(execute=@n6.else, Condition=Get bIsCrouching())
   > Set RB alpha(RB alpha="Clamp (Float)"(Value=Get crouchingMoving(), Min=0.6, Max=1))
   Set RB alpha(execute=@n7.else, RB alpha=1)
   n8 = Get SkelMesh()
   Set leg_l_rot(
    execute=@n2.then_1
    leg_l_rot:double=GetAlphaFromReferenceTransformDelta(SkeletalMeshComponent=@n8, BoneName=pelvis, BaseName=thigh_l, InputRange=(X=0,Y=200), TwistAxis="0,0,1")
   ) -> (Output_Get:double)
   > Set leg_r_rot(leg_r_rot:double=GetAlphaFromReferenceTransformDelta(SkeletalMeshComponent=@n8, BoneName=pelvis, BaseName=thigh_r, InputRange=(X=0,Y=200), TwistAxis="0,0,1")) -> (Output_Get:double)
   n9 = Get bIsMoving2D()
   > n10 = Branch(Condition=AND Boolean(A=Get bIsDBNO(), B=NOT Boolean(A=@n9)))
   n11 = Get leg_l_rot()
   Set leg_l_rot(execute=[@n13.then, @n10.then], leg_l_rot=float * float(A=@n11, B=0.4) {is=title K2Node_PromotableOperator})
   n12 = Get leg_r_rot()
   > Set leg_r_rot(leg_r_rot=float * float(A=@n12, B=0.5))
   n13 = Branch(execute=@n10.else, Condition=Get bIsCrouching())
   Set leg_l_rot(execute=@n13.else, leg_l_rot=@n11)
   > Set leg_r_rot(leg_r_rot=@n12)
   n14 = Get SkelMesh()
   Set headUp_BackBlink(
    execute=@n2.then_3
    headUp_BackBlink:double=float * float(
     A=GetAlphaFromReferenceTransformDelta(SkeletalMeshComponent=@n14, BoneName=spine_05, BaseName=head, InputRange=(X=-35,Y=55), TwistAxis="0,0,1")
     B:bool=Get isWearingBackbling()
    )
   ) -> (Output_Get:double)
   n15 = GetAnimInstance(self=GetOwnerSkeletalMeshComponent())
   Is Valid(exec=@n2.then_4, InputObject=@n15)
   n16 = GetCurveValue(CurveName=CajunTaco)
   Set Is Cajuntaco Emote(execute=@n2.then_5, Is Cajuntaco Emote:bool="To Boolean (Integer)"!(InInt:int=Truncate(A=@n16)) -> (ReturnValue:bool)) -> (Output_Get:bool)
   > Set CustomEmoteFloat(CustomEmoteFloat:double=@n16) -> (Output_Get:double)
   n17 = CheckForBackbling()
   Set isWearingBackbling(
    execute=@n2.then_2
    isWearingBackbling=AND Boolean!(A=@n17, B=NOT Boolean(A=@n17.bIsBackblingACape), C:bool=NOT Boolean(A=@n17.bIsBackblingLong)) -> (ReturnValue)
   )
   > Set Is Wearing Cape(Is Wearing Cape=OR Boolean(A=@n17.bIsBackblingACape, B=@n17.bIsBackblingLong))
   n18 = Get bIsMoving2D()
   > Set isCrounchingNotMoving(isCrounchingNotMoving:bool=AND Boolean!(A=Get bIsCrouching(), B=NOT Boolean(A=@n18)) -> (ReturnValue)) -> (Output_Get:bool)
   > n19 = Branch(Condition=Get bIsDBNO())
   > Set rotationRside(rotationRside:Rotator="60,-45,-45") -> (Output_Get:Rotator)
   Set rotationRside(execute=@n19.else, rotationRside="45,0,-45")
   n20 = Event OnCharacterCustomizationLoaded()
   > Set SkelMesh(SkelMesh=@n20.CustomBaseMesh)
   Event OnBeginCrouching()
   > Set isCrouchInterp(isCrouchInterp:double=1) -> (Output_Get:double)
   > Delay(Duration=0.3)
   > Set isCrouchInterp(isCrouchInterp=0)
   n21 = Montage_IsPlaying(self=@n15, Montage=/BRCosmetics/Animation/Game/MainPlayer/Emotes/CajunTaco/Followers/CMF/Emote_CajunTaco_CMF_Follower_M)
   n22 = Montage_IsPlaying(self=@n15, Montage=/BRCosmetics/Animation/Game/MainPlayer/Emotes/ChiveFlake/Frontend/Emote_ChiveFlake_CMF_Frontend_M)
   n23 = Montage_IsPlaying(self=@n15, Montage=/BRCosmetics/Animation/Game/MainPlayer/Emotes/CajunTaco/CMF/Emote_CajunTaco_CMF_M)
   Set Is Custom Emote Playing(Is Custom Emote Playing:bool=OR Boolean!(
     A=Montage_IsPlaying(self=@n15, Montage=/BRCosmetics/Animation/Game/MainPlayer/Emotes/ChiveFlake/Hero/CMF/Emote_ChiveFlake_Hero_CMF_M)
     B=Montage_IsPlaying(self=@n15, Montage=/BRCosmetics/Animation/Game/MainPlayer/Emotes/ChiveFlake/CMF/Emote_ChiveFlake_CMF_M)
     C:bool=@n21
     D:bool=@n23
     E:bool=@n22
     F:bool=@n22
    ) -> (ReturnValue)) -> (Output_Get:bool)
   > Set Is Cajuntaco Emote(Is Cajuntaco Emote=OR Boolean!(A=@n21, B=@n23) -> (ReturnValue))
   Montage_IsPlaying(self=@n15, Montage=/BRCosmetics/Animation/Game/MainPlayer/Emotes/PuzzleShed/CMF/Emote_PuzzleShed_CMF_M)
   Montage_IsPlaying(self=@n15, Montage=/BRCosmetics/Animation/Game/MainPlayer/Emotes/ChiveFlake/Hero/CMF/Frontend/Emote_ChiveFlake_Hero_CMF_FrontEnd_M)
   FInterpTo(Current=Get CustomEmoteFloat(), Target="To Float (Boolean)"(InBool=Get Is Custom Emote Playing()), DeltaTime=@n1.DeltaTimeX, InterpSpeed=1.5)
   "To Bool (Double)"!(Value:double=0) -> (ReturnValue:bool)
   Event OnBeginPlayingEmote() {is=title K2Node_Event}
   Event OnEndPlayingEmote() {is=title K2Node_Event}
   Set Is Cajuntaco Emote(Is Cajuntaco Emote=false)
  ]}
]
variables [
 crouchingMoving {type=number}
 CustomEmoteFloat {type=number}
 dbnoFloat {type=number}
 DeltaTimeLocal {type=number}
 Gravity Override Params {type=GravityOverrideParamsStruct}
 GravityOverride {type=Vector}
 headUp_BackBlink {type=number}
 Is Cajuntaco Emote {type=boolean}
 Is Custom Emote Playing {type=boolean}
 Is Wearing Cape {type=boolean}
 isCrouchInterp {type=number}
 isCrounchingNotMoving {type=boolean}
 isPickAxeEnabled {type=boolean}
 isWearingBackbling {type=boolean}
 leg_l_rot {type=number}
 leg_r_rot {type=number}
 RB alpha {type=number}
 rotationRside {type=Rotator}
 SkelMesh {type=SkeletalMeshComponent}
 skydiving {type=number}
 slidingPickaxe {type=boolean}
]
