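# Physics asset for the F_MED_ChiveFlake face accessory mesh, covering a hair ball and ponytail on a female character.
# All bodies are capsules, with spine, clavicle, upper arm, neck, head and hair base linked by free angular constraints.
# From the hair base, a limited joint drives the hair ball, a locked joint attaches the ponytail root, and three segments follow with widening swing limits and tight or locked twist.
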PhysicsAsset /BRCosmetics/Characters/Player/Female/Medium/Bodies/F_MED_ChiveFlake/Meshes/Parts/F_MED_ChiveFlake_FaceAcc_Physics
bodies {
 base_hair_ball [{shapeType=Capsule, length=0.505, radius=0.505, shapeName=base_hair_ball_capsule, rotation=<rot 0 -90 0>}]
 clavicle_l [{shapeType=Capsule, length=8, radius=4.5, shapeName=clavicle_l_capsule, center=<5 0 2>, rotation=<rot -90 0 0>}]
 clavicle_r [{shapeType=Capsule, length=8, radius=4.5, shapeName=clavicle_r_capsule, center=<-5 0 -2>, rotation=<rot -90 0 0>}]
 dyn_hair_ball_01 [{shapeType=Capsule, length=5, radius=5, center=<3 0 0>, rotation=<rot -90 0 0>}]
 dyn_ponytail_01 [{shapeType=Capsule, length=5, radius=10, center=<4 0 1.3345>, rotation=<rot -90 180 180>}]
 dyn_ponytail_02 [{shapeType=Capsule, length=5, radius=8, center=<4 0 0>, rotation=<rot -90 0 0>}]
 dyn_ponytail_03 [{shapeType=Capsule, length=5, radius=5, center=<4 0 0>, rotation=<rot -90 0 0>}]
 dyn_ponytail_Ad_00 [{shapeType=Capsule, length=5, radius=5, center=<4 0 0>, rotation=<rot -90 0 0>}]
 head [{shapeType=Capsule, length=8.8258, radius=12, shapeName=head_capsule, center=<8.8117 -1.5027 0>, rotation=<rot -90 0 0>}]
 neck_01 [{shapeType=Capsule, length=4, radius=9, shapeName=head_capsule, center=<2.6435 -2.3975 0>, rotation=<rot -90 0 0>}]
 neck_02 [{shapeType=Capsule, length=4, radius=9, shapeName=head_capsule, center=<2.6435 -2.3975 0>, rotation=<rot -90 0 0>}]
 spine_03 [{shapeType=Capsule, length=20, radius=13, shapeName=spine_05_capsule, center=<-6.4547 -4.1521 0>, rotation=<rot 0 -90 0>}]
 spine_04 [{shapeType=Capsule, length=20, radius=13, shapeName=spine_05_capsule, center=<-4.5002 -4.1521 0>, rotation=<rot 0 -90 0>}]
 spine_05 [{shapeType=Capsule, length=20, radius=13, shapeName=spine_05_capsule, center=<0 -4.1521 0>, rotation=<rot 0 -90 0>}]
 upperarm_l [{shapeType=Capsule, length=35, radius=7, shapeName=upperarm_l_capsule, center=<19 0 0>, rotation=<rot -90 0 0>}]
 upperarm_r [{shapeType=Capsule, length=35, radius=7, shapeName=upperarm_r_capsule, center=<-19 0 0>, rotation=<rot -90 0 0>}]
}
constraints [
 {bone1Name=dyn_hair_ball_01, bone2Name=base_hair_ball, swing1LimitDegrees=10, swing2LimitDegrees=10, twistLimitDegrees=15}
 {
  bone1Name=base_hair_ball
  bone2Name=head
  swing1LimitDegrees=45
  swing1Motion=Free
  swing2LimitDegrees=45
  swing2Motion=Free
  twistLimitDegrees=45
  twistMotion=Free
 }
 {
  bone1Name=head
  bone2Name=neck_02
  swing1LimitDegrees=45
  swing1Motion=Free
  swing2LimitDegrees=45
  swing2Motion=Free
  twistLimitDegrees=45
  twistMotion=Free
 }
 {
  bone1Name=neck_02
  bone2Name=neck_01
  swing1LimitDegrees=45
  swing1Motion=Free
  swing2LimitDegrees=45
  swing2Motion=Free
  twistLimitDegrees=45
  twistMotion=Free
 }
 {
  bone1Name=neck_01
  bone2Name=spine_05
  swing1LimitDegrees=45
  swing1Motion=Free
  swing2LimitDegrees=45
  swing2Motion=Free
  twistLimitDegrees=45
  twistMotion=Free
 }
 {
  bone1Name=dyn_ponytail_Ad_00
  bone2Name=dyn_hair_ball_01
  swing1LimitDegrees=1
  swing1Motion=Locked
  swing2LimitDegrees=1
  swing2Motion=Locked
  twistLimitDegrees=1
  twistMotion=Locked
 }
 {bone1Name=dyn_ponytail_01, bone2Name=dyn_ponytail_Ad_00, swing1LimitDegrees=10, swing2LimitDegrees=10, twistLimitDegrees=2}
 {bone1Name=dyn_ponytail_02, bone2Name=dyn_ponytail_01, swing1LimitDegrees=15, swing2LimitDegrees=15, twistLimitDegrees=2}
 {bone1Name=dyn_ponytail_03, bone2Name=dyn_ponytail_02, swing1LimitDegrees=30, swing2LimitDegrees=45, twistLimitDegrees=1, twistMotion=Locked}
 {
  bone1Name=clavicle_l
  bone2Name=spine_05
  swing1LimitDegrees=45
  swing1Motion=Free
  swing2LimitDegrees=45
  swing2Motion=Free
  twistLimitDegrees=45
  twistMotion=Free
 }
 {
  bone1Name=clavicle_r
  bone2Name=spine_05
  swing1LimitDegrees=45
  swing1Motion=Free
  swing2LimitDegrees=45
  swing2Motion=Free
  twistLimitDegrees=45
  twistMotion=Free
 }
 {
  bone1Name=upperarm_l
  bone2Name=clavicle_l
  swing1LimitDegrees=45
  swing1Motion=Free
  swing2LimitDegrees=45
  swing2Motion=Free
  twistLimitDegrees=45
  twistMotion=Free
 }
 {
  bone1Name=upperarm_r
  bone2Name=clavicle_r
  swing1LimitDegrees=45
  swing1Motion=Free
  swing2LimitDegrees=45
  swing2Motion=Free
  twistLimitDegrees=45
  twistMotion=Free
 }
 {
  bone1Name=spine_05
  bone2Name=spine_04
  swing1LimitDegrees=45
  swing1Motion=Free
  swing2LimitDegrees=45
  swing2Motion=Free
  twistLimitDegrees=45
  twistMotion=Free
 }
 {
  bone1Name=spine_04
  bone2Name=spine_03
  swing1LimitDegrees=45
  swing1Motion=Free
  swing2LimitDegrees=45
  swing2Motion=Free
  twistLimitDegrees=45
  twistMotion=Free
 }
]
