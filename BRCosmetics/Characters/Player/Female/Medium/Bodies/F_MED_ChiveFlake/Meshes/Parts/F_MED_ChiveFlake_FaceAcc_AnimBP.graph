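# Post-copy AnimBP for the ChiveFlake female face accessory, driving ponytail and hair motion around backblings and capes.
# Poses copied from parent and body meshes feed RigidBody simulation, or a look-at/constraint chain when short backbling is worn; emotes and frontend use fixed constraints with backbling.
# EventGraph derives corrective alphas from head-to-spine deltas, movement, stance and melee, plus gravity override.
# AnimDynamics finishes side hair and hair ball.
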AnimBlueprint /BRCosmetics/Characters/Player/Female/Medium/Bodies/F_MED_ChiveFlake/Meshes/Parts/F_MED_ChiveFlake_FaceAcc_AnimBP
target_skeleton=&/F_MED_ChiveFlake_FaceAcc_Skeleton
graphs [
 AnimGraph {nodes [
   n1 = "Transform (Modify) Bone - Bone: dyn_ponytail_Ad_00"(Alpha=Get headUp_Backpack() -> (headUp_Backpack:double), Translation="-2,0,0", Rotation="-25,0,0", Scale="1,1,1") {
    rotationMode=BMM_Additive
    rotationSpace=BCS_BoneSpace
    scaleMode=BMM_Additive
    scaleSpace=BCS_BoneSpace
    translationMode=BMM_Additive
    translationSpace=BCS_BoneSpace
    alphaBoolBlend {blendOption=Linear}
    boneToModify {boneName=dyn_ponytail_Ad_00}
    rotation {pitch=-25}
    translation {x=-2}
   }
   "Transform (Modify) Bone - Bone: base_hair_ball"(ComponentPose=@n1, Alpha=1, Translation="0,0,0", Rotation="-10,0,0", Scale="1,1,1") {
    rotationMode=BMM_Additive
    rotationSpace=BCS_BoneSpace
    scaleMode=BMM_Additive
    scaleSpace=BCS_BoneSpace
    translationMode=BMM_Additive
    alphaBoolBlend {blendOption=Linear}
    boneToModify {boneName=base_hair_ball}
    rotation {pitch=-10}
   }
   n2 = "Transform (Modify) Bone - Bone: dyn_ponytail_Ad_00"(ComponentPose=Local To Component(LocalPose=Use cached pose 'startCache'()), Alpha=1, Translation="0,0,0", Rotation="5,0,0", Scale="1,1,1") {
    rotationMode=BMM_Additive
    rotationSpace=BCS_BoneSpace
    alphaBoolBlend {blendOption=Linear}
    boneToModify {boneName=dyn_ponytail_Ad_00}
    rotation {pitch=5}
   }
   n3 = "Constraint - dyn_ponytail_04 (PP)"!(ComponentPose=@n2, Alpha=0.8, ConstraintWeights_0=0.5, ConstraintWeights_1:float=0.5) -> (Pose) {
    alphaBoolBlend {blendOption=Linear}
    boneToModify {boneName=dyn_ponytail_04}
    constraintSetup [{targetBone {boneName=spine_05}}, {targetBone {boneName=neck_01}}]
    constraintWeights [1,1]
   }
   n4 = "Constraint - dyn_ponytail_03 (PP)"!(ComponentPose=@n3, Alpha=0.8, ConstraintWeights_0=0.5, ConstraintWeights_1:float=0.5) -> (Pose) {
    alphaBoolBlend {blendOption=Linear}
    boneToModify {boneName=dyn_ponytail_03}
    constraintSetup [{targetBone {boneName=neck_02}}, {targetBone {boneName=neck_01}}]
    constraintWeights [1,1]
   }
   n5 = "Constraint - dyn_ponytail_02 (PP)"!(ComponentPose=@n4, Alpha=0.8, ConstraintWeights_0=0.5, ConstraintWeights_1:float=0.5) -> (Pose) {
    alphaBoolBlend {blendOption=Linear}
    boneToModify {boneName=dyn_ponytail_02}
    constraintSetup [{targetBone {boneName=neck_02}}, {targetBone {boneName=head}}]
    constraintWeights [1,1]
   }
   n6 = "Constraint - dyn_ponytail_01 (P)"(ComponentPose=@n5, Alpha=1, ConstraintWeights_0=1) {
    alphaBoolBlend {blendOption=Linear}
    boneToModify {boneName=dyn_ponytail_01}
    constraintSetup [{targetBone {boneName=head}}]
    constraintWeights [1]
   }
   n7 = Component To Local(ComponentPose=@n6)
   n8 = Get isWearingBackbling_ConstraintsAPPLIED() -> (isWearingBackbling_ConstraintsAPPLIED:bool)
   n9 = Look At - Bone: dyn_ponytail_01(ComponentPose=Local To Component(LocalPose=Use cached pose 'startCache'()), Alpha=Get headUp_LookAt() -> (headUp_LookAt:double)) {
    alphaBoolBlend {blendOption=Linear}
    boneToModify {boneName=dyn_ponytail_01}
    lookAtTarget {boneReference {boneName=spine_02}}
    lookAt_Axis {axis=<-1 0>}
    lookUp_Axis {axis {x=0, z=1}}
   }
   n10 = "Transform (Modify) Bone - Bone: dyn_ponytail_Ad_00"(ComponentPose=@n9, Alpha=Get headUp_Backpack(), Translation="0,0,0", Rotation="0,0,-40", Scale="1,1,0.1") {
    rotationMode=BMM_Additive
    scaleMode=BMM_Additive
    scaleSpace=BCS_BoneSpace
    translationMode=BMM_Additive
    alphaBoolBlend {blendOption=Linear}
    boneToModify {boneName=dyn_ponytail_Ad_00}
    rotation {roll=-40}
    scale {z=0.1}
   }
   n11 = "Transform (Modify) Bone - Bone: dyn_ponytail_01"(ComponentPose=@n10, Alpha=Get headUp_LookAt(), Translation="0,0,0", Rotation="0,0,-40", Scale="1,1,0.9") {
    rotationMode=BMM_Additive
    scaleMode=BMM_Additive
    scaleSpace=BCS_BoneSpace
    translationMode=BMM_Additive
    alphaBoolBlend {blendOption=Linear}
    boneToModify {boneName=dyn_ponytail_01}
    rotation {roll=-40}
    scale {z=0.9}
   }
   n12 = Get headUp_LookAt()
   n13 = "Constraint - dyn_ponytail_03 (P)"(ComponentPose=@n11, Alpha=@n12, ConstraintWeights_0=1) {
    alphaBoolBlend {blendOption=Linear}
    boneToModify {boneName=dyn_ponytail_03}
    constraintSetup [{targetBone {boneName=spine_05}}]
    constraintWeights [1]
   }
   n14 = "Transform (Modify) Bone - Bone: dyn_hair_ball_02"(ComponentPose=@n13, Alpha=@n12, Translation="0,5,0", Rotation="0,0,0", Scale="1,1,1") {
    rotationMode=BMM_Additive
    scaleMode=BMM_Additive
    scaleSpace=BCS_BoneSpace
    translationMode=BMM_Additive
    alphaBoolBlend {blendOption=Linear}
    boneToModify {boneName=dyn_hair_ball_02}
    translation {y=5}
   }
   n15 = "Transform (Modify) Bone - Bone: dyn_ponytail_03"(ComponentPose=@n14, Alpha=@n12, Translation="0,-4,-8", Rotation="0,0,0", Scale="1,1,1") {
    rotationMode=BMM_Additive
    scaleMode=BMM_Additive
    scaleSpace=BCS_BoneSpace
    translationMode=BMM_Additive
    alphaBoolBlend {blendOption=Linear}
    boneToModify {boneName=dyn_ponytail_03}
    translation {y=-4, z=-8}
   }
   n16 = "Constraint - dyn_ponytail_03 (P)"(ComponentPose=@n15, Alpha=Get headDown_Backpack() -> (headDown_Backpack:double), ConstraintWeights_0=1) {
    alphaBoolBlend {blendInTime=0.2, blendOption=Linear, blendOutTime=0.2}
    boneToModify {boneName=dyn_ponytail_03}
    constraintSetup [{targetBone {boneName=spine_05}}]
    constraintWeights [1]
   }
   n17 = "Transform (Modify) Bone - Bone: dyn_ponytail_01"(
    ComponentPose=@n16
    Alpha=Get headDown_BackpackMain() -> (headDown_BackpackMain:double)
    Translation="0,0,0"
    Rotation="0,0,9.5"
    Scale="1,1,1"
   ) {
    rotationMode=BMM_Additive
    scaleMode=BMM_Additive
    scaleSpace=BCS_BoneSpace
    translationMode=BMM_Additive
    alphaBoolBlend {blendOption=Linear}
    boneToModify {boneName=dyn_ponytail_01}
    rotation {roll=9.5}
   }
   n18 = Blend Poses by bool(BlendPose_0=Component To Local(ComponentPose=@n17), BlendPose_1=Use cached pose 'RBcache'(), bActiveValue=Get isWearingBackbling()) {blendType=Linear}
   n19 = Anim Dynamics - Bone: dyn_hair_side_l(
    ComponentPose=Local To Component(LocalPose=Blend Poses by bool(BlendPose_0=@n7, BlendPose_1=@n18, BlendTime_0=0.2, BlendTime_1=0.2, bActiveValue=@n8) {blendType=Linear})
    Alpha=1
    ExternalForce="0,0,0"
   ) {
    angularDampingOverride=1
    angularSpringConstant=5000
    bAngularSpring=true
    bOverrideAngularDamping=true
    bUseGravityOverride=true
    gravityScale=0.35
    simulationSpace=RootRelative
    alphaBoolBlend {blendOption=Linear}
    boundBone {boneName=dyn_hair_side_l}
    componentAppliedLinearAccClamp=<2500 2500 1500>
    componentLinearAccScale=<1 1 1>
    componentLinearVelScale=<1 1 1>
    gravityOverride {z=-980}
    physicsBodyDefinitions [{
      boundBone {boneName=dyn_hair_side_l}
      boxExtents=<35 35 35>
      constraintSetup {angularLimitsMax=<15 0 30>, angularLimitsMin=<-15 -30 0>, angularTarget=<1 0 0>}
      localJointOffset=<3 0 0>
     }]
   }
   n20 = Anim Dynamics - Bone: dyn_hair_side_r(ComponentPose=@n19, Alpha=1, ExternalForce="0,0,0") {
    angularDampingOverride=1
    angularSpringConstant=5000
    bAngularSpring=true
    bOverrideAngularDamping=true
    bUseGravityOverride=true
    gravityScale=0.35
    simulationSpace=RootRelative
    alphaBoolBlend {blendOption=Linear}
    boundBone {boneName=dyn_hair_side_r}
    componentAppliedLinearAccClamp=<2500 2500 1500>
    componentLinearAccScale=<1 1 1>
    componentLinearVelScale=<1 1 1>
    gravityOverride {z=-980}
    physicsBodyDefinitions [{
      boundBone {boneName=dyn_hair_side_r}
      boxExtents=<35 35 35>
      constraintSetup {angularLimitsMax=<15 0 30>, angularLimitsMin=<-15 -30 0>, angularTarget=<1 0 0>}
      localJointOffset=<3 0 0>
     }]
   }
   n21 = Anim Dynamics - Bone: dyn_hair_ball_02(ComponentPose=@n20, Alpha=1, ExternalForce="0,0,0") {
    angularDampingOverride=1
    angularSpringConstant=5000
    bAngularSpring=true
    bOverrideAngularDamping=true
    bUseGravityOverride=true
    gravityScale=0.35
    simulationSpace=RootRelative
    alphaBoolBlend {blendOption=Linear}
    boundBone {boneName=dyn_hair_ball_02}
    componentAppliedLinearAccClamp=<2500 2500 1500>
    componentLinearAccScale=<1 1 1>
    componentLinearVelScale=<1 1 1>
    gravityOverride {z=-980}
    physicsBodyDefinitions [{
      boundBone {boneName=dyn_hair_ball_02}
      boxExtents=<20 20 20>
      constraintSetup {angularLimitsMax=<15 5 15>, angularLimitsMin=<-15 -15 -15>, angularTarget=<1 0 0>}
      localJointOffset=<3 0 0>
     }]
   }
   Output Pose(Result=Component To Local(ComponentPose=@n21)) {layerGroup=None}
   n22 = Layered blend per bone(
    BasePose=Copy Pose From Mesh() {bCopyCurves=true, bUseAttachedParent=true}
    BlendPoses_0=Copy Pose From Mesh(SourceMeshComponent=Get Body_Mesh())
   ) {layerSetup [{branchFilters [{boneName=clavicle_l}, {boneName=clavicle_r}, {boneName=upperarm_r}, {boneName=upperarm_l}]}]}
   Save cached pose 'startCache'(Pose=@n22) {cacheName=startCache}
   n23 = "Transform (Modify) Bone - Bone: base_hair_ball"!(
    ComponentPose=Local To Component(LocalPose=Use cached pose 'startCache'())
    bAlphaBoolEnabled:bool=True
    Translation="0,0,0"
    Rotation="10,0,0"
    Scale="1,1,1"
   ) -> (Pose) {
    alphaInputType=Bool
    rotationMode=BMM_Additive
    rotationSpace=BCS_BoneSpace
    scaleMode=BMM_Additive
    scaleSpace=BCS_BoneSpace
    translationMode=BMM_Additive
    alphaBoolBlend {blendOption=Linear}
    boneToModify {boneName=base_hair_ball}
    rotation {pitch=10}
   }
   n24 = "Transform (Modify) Bone - Bone: base_hair_ball"!(ComponentPose=@n23, Alpha:float=Get headUp_Cape() -> (headUp_Cape:double), Translation="0,0,0", Rotation="60,0,0", Scale="1,1,1") -> (Pose) {
    rotationMode=BMM_Additive
    rotationSpace=BCS_BoneSpace
    translationMode=BMM_Additive
    translationSpace=BCS_BoneSpace
    alphaBoolBlend {blendOption=Linear}
    boneToModify {boneName=base_hair_ball}
    rotation {pitch=60}
   }
   n25 = RigidBody!(ComponentPose=@n24, Alpha, OverrideWorldGravity:Vector=Get GravityOverride(), ExternalForce) -> (Pose) {
    bClampLinearTranslationLimitToRefPose=true
    bOverrideWorldGravity=true
    overridePhysicsAsset=&/F_MED_ChiveFlake_FaceAcc_Physics
    simulationSpace=BaseBoneSpace
    alphaBoolBlend {blendOption=Linear}
    baseBoneRef {boneName=neck_02}
    componentAppliedLinearAccClamp=<3000 3000 3000>
    componentLinearAccScale=<0.5 0.5 0.5>
    componentLinearVelScale=<0.5 0.5 0.5>
    overrideWorldGravity {z=-980}
   }
   Save cached pose 'RBcache'(Pose=Component To Local(ComponentPose=@n25)) {cacheName=RBcache}
  ]}
 EventGraph {nodes [
   Event OnCharacterCustomizationLoaded()
   > n1 = Cast To PlayerPawn_Athena(Object=TryGetPawnOwner())
   > Set Body_Mesh(Body_Mesh=GetSkeletalMeshForPartType(self=@n1))
   n2 = Event BlueprintUpdateAnimation()
   > Set deltatime(deltatime:double=@n2.DeltaTimeX) -> (Output_Get:double)
   > n3 = Sequence!(execute) -> (then_0, then_1, then_2:Exec)
   > n4 = Gravity Override State Function DEPRECATED(CharacterPartAnimInstance=Self-Reference(), GameTimeInSeconds=GetGameTimeInSeconds(), GravityOverrideParams=Get Gravity Override Params())
   > Set GravityOverride(GravityOverride=@n4)
   n5 = Cast To PlayerPawn_Athena(execute=@n3.then_1, Object=TryGetPawnOwner())
   > Set SkelMesh(SkelMesh=GetSkeletalMeshForPartType(self=@n5))
   n6 = Get SkelMesh()
   n7 = GetAlphaFromReferenceTransformDelta(SkeletalMeshComponent=@n6, BoneName=spine_05, BaseName=head, InputRange=(X=-35,Y=75), TwistAxis="0,0,1")
   > Set headUp_Cape(headUp_Cape:double=float * float(A=@n7, B:bool=Get Is Wearing Cape() -> (Is Wearing Cape:bool)) {is=title K2Node_PromotableOperator}) -> (Output_Get:double)
   n8 = Get isWearingBackbling()
   > Set headUp_Backpack(headUp_Backpack:double="Clamp (Float)"(Value=float * float(A=@n7, B=@n8), Min=0, Max=0.5)) -> (Output_Get:double)
   > Set headUp_LookAt(headUp_LookAt:double=GetAlphaFromReferenceTransformDelta(SkeletalMeshComponent=@n6, BoneName=spine_05, BaseName=head, InputRange=(X=-35,Y=25), TwistAxis="0,0,1")) -> (Output_Get:double)
   > n9 = Set headDown_BackpackMain(headDown_BackpackMain:double=GetAlphaFromReferenceTransformDelta(SkeletalMeshComponent=@n6, BoneName=spine_01, BaseName=head, InputRange=(X=15,Y=0), TwistAxis="0,0,1")) -> (Output_Get:double)
   n10 = Get bIsMoving2D()
   > Set isMoving2D(isMoving2D:double=FInterpTo(
     Current=Get isMoving2D() -> (isMoving2D:double)
     Target="To Float (Boolean)"(InBool=@n10)
     DeltaTime=Get deltatime() -> (deltatime:double)
     InterpSpeed=2
    )) -> (Output_Get:double)
   > n11 = Branch(Condition=OR Boolean(A=Get bIsTacticalSprinting(), B=Get bIsCrouching()))
   > Set headDown_Backpack(headDown_Backpack:double=1) -> (Output_Get:double)
   Set headDown_Backpack(execute=@n11.else, headDown_Backpack=float * float(A=@n9, B:double=Get isMoving2D()))
   n12 = Get bIsPlayingMeleeAnim()
   > n13 = Set isMelee(isMelee:double=FInterpTo(Current=Get isMelee() -> (isMelee:double), Target="To Float (Boolean)"(InBool=@n12), DeltaTime=Get deltatime(), InterpSpeed=0.5)) -> (Output_Get:double)
   > Set headDown_Backpack(headDown_Backpack=@n13)
   n14 = Get isWearingBackbling()
   n15 = Get bIsPlayingEmote()
   n16 = Get bIsInFrontEnd()
   n17 = Get isWearingBackbling()
   > Set isWearingBackbling_ConstraintsAPPLIED(isWearingBackbling_ConstraintsAPPLIED:bool=OR Boolean(A=AND Boolean(A=@n15, B=@n14), B=AND Boolean(A=@n16, B=@n17))) -> (Output_Get:bool)
   n18 = CheckForBackbling()
   Set isWearingBackbling(
    execute=@n3.then_2
    isWearingBackbling=AND Boolean!(A=@n18, B=NOT Boolean(A=@n18.bIsBackblingACape), C:bool=NOT Boolean(A=@n18.bIsBackblingLong)) -> (ReturnValue)
   )
   > Set Is Wearing Cape(Is Wearing Cape=OR Boolean(A=@n18.bIsBackblingACape, B=@n18.bIsBackblingLong))
   > Set isCapeBP_ON(isCapeBP_ON:bool=OR Boolean(A=Get isWearingBackbling(), B=Get Is Wearing Cape())) -> (Output_Get:bool)
   Get isCapeBP_ON() -> (isCapeBP_ON:bool)
  ]}
]
variables [
 Body_Mesh {type=SkeletalMeshComponent}
 deltatime {type=number}
 Gravity Override Params {type=GravityOverrideParamsStruct}
 GravityOverride {type=Vector}
 headDown_Backpack {type=number}
 headDown_BackpackMain {type=number}
 headUp_Backpack {type=number}
 headUp_Cape {type=number}
 headUp_LookAt {type=number}
 Is Wearing Cape {type=boolean}
 isCapeBP_ON {type=boolean}
 isMelee {type=number}
 isMoving2D {type=number}
 isWearingBackbling {type=boolean}
 isWearingBackbling_ConstraintsAPPLIED {type=boolean}
 SkelMesh {type=SkeletalMeshComponent}
]
